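# Conversation database for The Order (BinaryTwin) NPC, entered through the TheOrder entry tag.
# A single speech node picks the greeting: a default line, or contextual lines for the first-ever conversation, specific player outfits, or this NPC's match conversation count.
# Outfit and first-meeting priorities come from named curve rows.
# The speech then fans out to link nodes for shared quests, basic and random services, dialogue, and Season 39 story and weekly quest conversations.
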ConversationDatabase /NPCLibrary/NPCs/BinaryTwin/BinaryTwin/NPCLibrary_Conversation_BinaryTwin
entry_points [{tag=Conversation.Entry.TheOrder, destinations [#1]}]
nodes [
 ConversationEntryPointNode_7 {
  id=#1
  class=ConversationEntryPointNode
  outputs [#2]
  properties {entryTag {tagName=Conversation.Entry.TheOrder}}
 }
 Speech_C_4 {
  id=#2
  class=Speech_C
  outputs [
   #3
   #4
   #5
   #6
   #7
   #8
  ]
  properties {
   nodeName=""
   generalConfig {
    defaultMessage=Perimeter secure.
    contextualMessages [
     {
      message=Call me The Order.
      contextRequirements [{requirement {class=FortControllerRequirement_IsFirstEverConversationWithNPC}}]
      priority {curve {rowName=Tandem.Speech.Priority.FirstEverInteraction}}
     }
     {
      message="Good to see you, Jones."
      contextRequirements [{requirement {
         class=FortControllerRequirement_HasCID
         properties {allowedCIds [&Character_UndergroundRebel_Fashion]}
        }}]
      priority {curve {rowName=Tandem.Speech.Priority.OutfitReaction}}
     }
     {
      message=Jones has told me a lot about you.
      contextRequirements [{requirement {class=FortControllerRequirement_HasCID, properties {allowedCIds [&Character_BrightDisk]}}}]
      priority {curve {rowName=Tandem.Speech.Priority.OutfitReaction}}
     }
     {
      message="I don't know what your game is, but you won't win."
      contextRequirements [{requirement {class=FortControllerRequirement_HasCID, properties {allowedCIds [/BRCosmetics/Athena/Items/Cosmetics/Characters/Character_SpacePlunge]}}}]
      priority {curve {rowName=Tandem.Speech.Priority.OutfitReaction}}
     }
     {
      message=What dark place did you and your master come from?
      contextRequirements [{requirement {class=FortControllerRequirement_HasCID, properties {allowedCIds [/BRCosmetics/Athena/Items/Cosmetics/Characters/Character_MincePounce]}}}]
      priority {curve {rowName=Tandem.Speech.Priority.OutfitReaction}}
     }
     {
      message=This Island is in danger.
      contextRequirements [{requirement {class=FortControllerRequirement_CheckNPCMatchConversationCount, properties {nPCUniqueIdTag {tagName=AISpawnerData.Type.Tandem.TheOrder}}}}]
      priority {curve {rowName=None}}
     }
     {
      message=Proud to be one of the Seven.
      contextRequirements [{requirement {class=FortControllerRequirement_CheckNPCMatchConversationCount, properties {nPCUniqueIdTag {tagName=AISpawnerData.Type.Tandem.TheOrder}}}}]
      priority {curve {rowName=None}}
     }
     {
      message=Can I trust you?
      contextRequirements [{requirement {class=FortControllerRequirement_CheckNPCMatchConversationCount, properties {nPCUniqueIdTag {tagName=AISpawnerData.Type.Tandem.TheOrder}}}}]
      priority {curve {rowName=None}}
     }
    ]
   }
  }
 }
 ConversationLinkNode_4 {
  id=#8
  class=ConversationLinkNode
  properties {remoteEntryTag {tagName=Conversation.Entry.Season39.Weekly.W13.Q01.BinaryTwin}}
 }
 ConversationLinkNode_3 {id=#7, class=ConversationLinkNode, properties {remoteEntryTag {tagName=Conversation.Entry.TheOrder.Dialogue}}}
 ConversationLinkNode_2 {
  id=#6
  class=ConversationLinkNode
  properties {remoteEntryTag {tagName=Conversation.Entry.Season39.BinaryTwin.Story.02.01}}
 }
 ConversationLinkNode_7 {id=#5, class=ConversationLinkNode, properties {remoteEntryTag {tagName=Conversation.Entry.RandomServices}}}
 ConversationLinkNode_0 {id=#4, class=ConversationLinkNode, properties {remoteEntryTag {tagName=Conversation.Entry.BasicServices}}}
 ConversationLinkNode_1 {id=#3, class=ConversationLinkNode, properties {remoteEntryTag {tagName=Conversation.Entry.NPCLibrary.Quests}}}
]
