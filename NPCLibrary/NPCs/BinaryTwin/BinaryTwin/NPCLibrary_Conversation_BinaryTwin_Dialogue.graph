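# Conversation database for the BinaryTwin NPC's lore dialogue, entered through the Conversation.Entry.TheOrder.Dialogue tag.
# An "Ask about..." speech node opens a menu with two topics, The Seven and The Zero Point, plus a Back option.
# Each topic plays a linear chain of Speech nodes about the scattered Seven or the shattered Zero Point, ending in a ReturnToStart node labelled BACK.
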ConversationDatabase /NPCLibrary/NPCs/BinaryTwin/BinaryTwin/NPCLibrary_Conversation_BinaryTwin_Dialogue
entry_points [{tag=Conversation.Entry.TheOrder.Dialogue, destinations [#1]}]
nodes [
 ConversationEntryPointNode_7 {
  id=#1
  class=ConversationEntryPointNode
  outputs [#2]
  properties {entryTag {tagName=Conversation.Entry.TheOrder.Dialogue}}
 }
 Speech_C_0 {
  id=#2
  class=Speech_C
  outputs [#3, #4, #5]
  properties {nodeName="", softTaskIcon=&T_UI_NPC_question, generalConfig {defaultMessage=What do you need to know?}}
  sub_nodes [ConversationChoiceNode_5 {class=ConversationChoiceNode, properties {defaultChoiceDisplayText=Ask about...}}]
 }
 Speech_C_1 {
  id=#5
  class=Speech_C
  outputs [#6]
  properties {
   nodeName=""
   softTaskIcon=&T_UI_NPC_question
   generalConfig {defaultMessage=I still can't believe the Zero Point is shattered to pieces.}
  }
  sub_nodes [ConversationChoiceNode_2 {class=ConversationChoiceNode, properties {defaultChoiceDisplayText=The Zero Point}}]
 }
 Speech_C_2 {
  id=#6
  class=Speech_C
  outputs [#7]
  properties {
   nodeName=""
   softTaskIcon=&T_UI_NPC_question
   generalConfig {defaultMessage="Perhaps it was safer in that form... at least, until now."}
  }
 }
 Speech_C_3 {
  id=#7
  class=Speech_C
  outputs [#8]
  properties {
   nodeName=""
   softTaskIcon=&T_UI_NPC_question
   generalConfig {defaultMessage="Its full power, in anyone's hands, would be disastrous."}
  }
 }
 Speech_C_4 {
  id=#8
  class=Speech_C
  outputs [#9]
  properties {nodeName="", softTaskIcon=&T_UI_NPC_question, generalConfig {defaultMessage="We have to protect it. "}}
 }
 ReturnToStart_C_0 {
  id=#9
  class=ReturnToStart_C
  properties {softTaskIcon=&T_UI_NPC_Back}
  sub_nodes [ConversationChoiceNode_3 {class=ConversationChoiceNode, properties {defaultChoiceDisplayText=BACK}}]
 }
 Back_C_0 {
  id=#4
  class=Back_C
  sub_nodes [ConversationChoiceNode_7 {class=ConversationChoiceNode, properties {defaultChoiceDisplayText=BACK}}]
 }
 Speech_C_7 {
  id=#3
  class=Speech_C
  outputs [#10]
  properties {
   nodeName=""
   softTaskIcon=&T_UI_NPC_question
   generalConfig {defaultMessage="We are scattered. Some of us, lost."}
  }
  sub_nodes [ConversationChoiceNode_0 {class=ConversationChoiceNode, properties {defaultChoiceDisplayText=The Seven}}]
 }
 Speech_C_15 {
  id=#10
  class=Speech_C
  outputs [#11]
  properties {
   nodeName=""
   softTaskIcon=&T_UI_NPC_question
   generalConfig {defaultMessage=I have to believe that more of us still stand.}
  }
 }
 Speech_C_14 {
  id=#11
  class=Speech_C
  outputs [#12]
  properties {
   nodeName=""
   softTaskIcon=&T_UI_NPC_question
   generalConfig {defaultMessage="To defeat the Dark Voyager, we will need all the help we can get."}
  }
 }
 ReturnToStart_C_23 {
  id=#12
  class=ReturnToStart_C
  properties {softTaskIcon=&T_UI_NPC_Back}
  sub_nodes [ConversationChoiceNode_1 {class=ConversationChoiceNode, properties {defaultChoiceDisplayText=BACK}}]
 }
]
